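# Physics asset for the FairyFlex pickaxe, a cosmetic melee weapon axe mesh.
# A single capsule body on the Melee_TwoHanded_Handle bone, offset and rotated from the bone origin, provides the weapon's entire simplified collision shape.
# It has no constraints, so there are no jointed or separately simulated parts.
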PhysicsAsset /BRCosmetics/Weapons/FORT_Melee/Pickaxe_FairyFlex/Meshes/FairyFlex_Axe_PhysicsAsset
bodies {Melee_TwoHanded_Handle [{shapeType=Capsule, length=32.8051, radius=17.4095, center=<-0.0106 -4.8885 13.5864>, rotation=<rot 11.899 -90 0.0316>}]}
constraints []
